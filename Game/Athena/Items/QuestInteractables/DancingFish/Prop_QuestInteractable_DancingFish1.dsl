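; Quest prop of a singing fish that dances to music when nearby players emote, crediting them with quest progress.
; The server tracks trigger overlaps and polls candidates on a timer, sending quest updates to emoting players.
; Replicated bPlayerDancing drives the looping montage and music.
; Outside Creative mode, clients watch the quest manager after a hard-coded load delay and play local success effects once the objective completes.
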(event EventBeginPlay
  (bind _returnvalue (Fortnite|GetLocalFortPlayerControllers true))
  (bind _output (Utilities|Array|Get(acopy) _returnvalue))
  (bind _dancetriggeroverlap (Variables|Default|GetDanceTriggerOverlap))
  ;[1 1 1] Bind Overlaps
  (Networking|SwitchHasAuthority
    (:Authority
      (Collision|BindEventtoOnComponentBeginOverlap _dancetriggeroverlap (delegate Custom|PawnNearFish))
      (Collision|BindEventtoOnComponentEndOverlap _dancetriggeroverlap (delegate Custom|PawnLeftFishArea)))
    (:Remote
      _))
  ;[1 1 1] Check if In Creative Mode or not
  (bind _asfort_game_state_athena (Utilities|Casting|CastToFortGameStateAthena (Game|GetGameState)))
  (if (Variables|/Script/FortniteGame.FortGameStateAthena|GetbPlaylistDataIsLoaded _asfort_game_state_athena)
    (bind _cancompletequest (Class|AthenaQuestFunctionLibrary|AthenaCheckCreativeMode _asfort_game_state_athena))
    (if _cancompletequest
      (Variables|QuestStuff|SetIsCreativeMode false)
      (else
        (Variables|QuestStuff|SetIsCreativeMode true)))
    (Networking|SwitchHasAuthority
      (:Remote
        ;[1 0.5334 0] Hack to Wait for Quest Manager to Load
        (Utilities|FlowControl|Delay 5.0)
        ;[1 1 1] CLIENT: Bind Event for When Quest Gets Updated
        (bind _asfort_player_controller_athena (Utilities|Casting|CastToFortPlayerControllerAthena _output))
        (bind (_questisvalid _questcompleted _objectivecompleted) (Default|AthenaCheckQuestandObjectiveCompletion _output (Variables|QuestStuff|GetQuestItem) (Variables|QuestStuff|GetObjectiveBackendName)))
        (if _questisvalid
          (if _objectivecompleted
            _
            (elif (Variables|QuestStuff|GetIsCreativeMode)
              _
              (else
                (Quests|BindEventtoOnQuestsUpdated (Quests|GetQuestManager _asfort_player_controller_athena "Invalid") (EventDispatchers|CreateEvent :!event Custom|QuestUpdated))
                (Variables|QuestStuff|SetQuestLocallyActive true)))))))
    (else
      (Playlist|BindEventtoOnCurrentPlaylistLoaded _asfort_game_state_athena (delegate Custom|PlaylistLoaded)))))

(event Custom|PlaylistLoaded
  (params (PlaylistName :type Name) (PlaylistContextTags :type Gameplay%20Tag%20Container%20Structure%20%28by%20ref%29))
  (bind _returnvalue (Fortnite|GetLocalFortPlayerControllers true))
  (bind _output (Utilities|Array|Get(acopy) _returnvalue))
  ;[1 1 1] Check if In Creative Mode or not
  (bind _cancompletequest (Class|AthenaQuestFunctionLibrary|AthenaCheckCreativeMode (Game|GetGameState)))
  (if _cancompletequest
    (Variables|QuestStuff|SetIsCreativeMode false)
    (else
      (Variables|QuestStuff|SetIsCreativeMode true)))
  (Networking|SwitchHasAuthority
    (:Remote
      ;[1 0.5334 0] Hack to Wait for Quest Manager to Load
      (Utilities|FlowControl|Delay 5.0)
      ;[1 1 1] CLIENT: Bind Event for When Quest Gets Updated
      (bind _asfort_player_controller_athena (Utilities|Casting|CastToFortPlayerControllerAthena _output))
      (bind (_questisvalid _questcompleted _objectivecompleted) (Default|AthenaCheckQuestandObjectiveCompletion _output (Variables|QuestStuff|GetQuestItem) (Variables|QuestStuff|GetObjectiveBackendName)))
      (if _questisvalid
        (if _objectivecompleted
          _
          (elif (Variables|QuestStuff|GetIsCreativeMode)
            _
            (else
              (Quests|BindEventtoOnQuestsUpdated (Quests|GetQuestManager _asfort_player_controller_athena "Invalid") (EventDispatchers|CreateEvent :!event Custom|QuestUpdated))
              (Variables|QuestStuff|SetQuestLocallyActive true))))))))

(event Custom|PawnNearFish
  (params (OverlappedComponent :type Primitive%20Component%20Object%20Reference) (OtherActor :type Actor%20Object%20Reference) (OtherComp :type Primitive%20Component%20Object%20Reference) (OtherBodyIndex :type Integer) (bFromSweep :type Boolean) (SweepResult :type Hit%20Result%20Structure%20%28by%20ref%29))
  ;[1 1 1] Gather Players Near Fish
  (bind _returnvalue (Utilities|Time|SetTimerbyEvent (EventDispatchers|CreateEvent :!event Custom|CheckForDancingAtFish) 0.25 true))
  (Variables|Default|SetDanceCheckTimer _returnvalue)
  (Utilities|Array|AddUnique (Variables|Default|GetDancerCandidatesAtFish) OtherActor))

(event Custom|QuestUpdated
  ;[1 1 1] CLIENT: After Quest Update - check Local Clients to give rewards
  (Development|PrintString "QuestUpdated" false false "(R=0.309646,G=1.000000,B=0.000000,A=1.000000)" 5.0)
  (if (Variables|QuestStuff|GetQuestLocallyActive)
    (bind (_questisvalid _questcompleted _objectivecompleted) (Default|AthenaCheckQuestandObjectiveCompletion (Utilities|Array|Get(acopy) (Fortnite|GetLocalFortPlayerControllers true)) (Variables|QuestStuff|GetQuestItem) (Variables|QuestStuff|GetObjectiveBackendName)))
    (if (or _questcompleted _objectivecompleted)
      (Quests|UnbindEventfromOnQuestsUpdated (Quests|GetQuestManager (stale EventBeginPlay _asfort_player_controller_athena) "Invalid") (EventDispatchers|CreateEvent :!event Custom|QuestUpdated))
      (Variables|QuestStuff|SetQuestLocallyActive false)
      (Utilities|FlowControl|Delay 0.2)
      (CallFunction|PlayQuestRewardFX))))

(event Custom|PawnLeftFishArea
  (params (OverlappedComponent :type Primitive%20Component%20Object%20Reference) (OtherActor :type Actor%20Object%20Reference) (OtherComp :type Primitive%20Component%20Object%20Reference) (OtherBodyIndex :type Integer))
  ;[1 1 1] Remove Players Leaving Area
  (Utilities|Array|RemoveItem (Variables|Default|GetDancerCandidatesAtFish) OtherActor)
  (if (<= (Utilities|Array|Length (Variables|Default|GetDancerCandidatesAtFish)) 0)
    (Utilities|Time|ClearandInvalidateTimerbyHandle (Variables|Default|GetDanceCheckTimer))
    (Variables|Default|SetbPlayerDancing false)
    (else
      (CallFunction|CheckforDancingatFish))))

(event Custom|CheckForDancingAtFish
  ;[1 1 1] Check for Dancing and Try to Give Reward to Dancers
  (bind _array_element (Utilities|Array|ForEachLoopwithBreak (Variables|Default|GetDancerCandidatesAtFish))
    (:LoopBody
      (bind _asathena_player_controller (Utilities|Casting|CastToAthena_PlayerController (Pawn|GetController (Variables|Actor|GetInstigator _array_element))))
      (if (Emote|IsPlayingEmote _asathena_player_controller)
        (Variables|Default|SetbPlayerDancing true)
        (Utilities|Array|AddUnique (Variables|Default|GetDancersForReward) _asathena_player_controller)
        (else
          (Variables|Default|SetbPlayerDancing false))))
    (:Completed
      (if (Variables|Default|GetbPlayerDancing)
        (CallFunction|DanceFishyDance)
        (Class|AthenaQuestFunctionLibrary|AthenaGiveQuestUpdatetoPlayers :InQuestItemRef (Variables|QuestStuff|GetQuestItem) :ObjectiveBackendName (Variables|QuestStuff|GetObjectiveBackendName) :PlayerControllersForUpdate (Variables|Default|GetDancersForReward))
        (Utilities|Array|Clear (Variables|Default|GetDancersForReward))))))

(event Custom|DanceFishyDance
  (bind _singingbass_music_cue (Variables|Default|GetSingingBass_Music_Cue))
  ;[1 1 1] Make Fish Dance
  (Utilities|FlowControl|DoOnce false :!label _label_2
    (:Completed
      (Animation|Montage|PlayMontage (Variables|Default|GetSK_Prop_Ranger_Fish_Budgeted) (Variables|Default|GetMontage) :!label _label
        (:then
          (Development|PrintString "FLOP!" true false)
          (if (not (Variables|Default|GetMusicAlreadyStarted))
            (Audio|Components|Audio|Play _singingbass_music_cue 0.0)
            (Variables|Default|SetMusicAlreadyStarted true)))
        (:OnCompleted
          (if (Variables|Default|GetbPlayerDancing)
            (goto _label)
            (else
              (Audio|Components|Audio|Stop _singingbass_music_cue)
              (Variables|Default|SetMusicAlreadyStarted false)
              (goto _label_2 Reset))))))))

(event Custom|PlayQuestRewardFX
  (bind _returnvalue (Fortnite|GetLocalFortPlayerControllers true))
  (bind _output (Utilities|Array|Get(acopy) _returnvalue))
  (bind _returnvalue_2 (Game|Player|GetPlayerPawn _output))
  ;[1 1 1] Quest Success FX
  (Development|PrintString "SUCCESS FX" false false "(R=1.000000,G=0.703857,B=0.040648,A=1.000000)" 5.0)
  (Effects|SpawnFXSystemAttached (Variables|QuestRewardFX|GetQuestSuccessEmitter) (Variables|/Script/Engine.Character|GetMesh _returnvalue_2))
  (Audio|PlaySoundatLocation (Variables|QuestRewardFX|GetQuestSuccessSound) (Transformation|GetActorLocation _returnvalue_2)) ; Quest Success Sound
  (Game|Feedback|ClientPlayForceFeedback (Utilities|Array|Get(acopy) (Fortnite|GetLocalFortPlayerControllers true)) (Variables|QuestRewardFX|GetForceFeedback) "None" false false false))

(fn OnRep_bPlayerDancing
  (if (Variables|Default|GetbPlayerDancing)
    (CallFunction|DanceFishyDance)))

(fn ConstructionScript)
